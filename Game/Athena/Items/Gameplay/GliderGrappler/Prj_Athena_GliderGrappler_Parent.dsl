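; Parent projectile for the Glider Grappler, which grapples the player and then deploys a glider.
; On grapple it applies an action-blocking gameplay effect and pushes a replicated glider cosmetic override.
; Jump apex, or a delay failsafe, removes the block, force-opens the glider unless disabled, and requests item removal when ammo is empty.
; Teleporting clears block and override and destroys the projectile; destruction pops the override unless skydiving.
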(event EventGrapple
  (bind _weapon_actor (Variables|Default|GetWeapon%20Actor))
  (bind _playerpawn (Variables|Default|GetPlayerPawn))
  ;[1 1 1] Call the Grapple but also set the blocking status for the glider version
  (Environment|ApplyGlobalEnvironmentGameplayEffecttoActor (Variables|Default|GetPlayerPawn) (Variables|Default|GetGE_GlideBlockActions))
  (Variables|Default|SetHas%20Deployed false)
  (EventDispatchers|BindEventtoOnPawnTeleported (Variables|Default|GetPlayerPawn) (delegate Custom|Teleported%20-%20Return%20abilities%20and%20stop))
  (|CallParentFunction)
  ;[1 1 1] Push Glider Override now so that it has time to replicate
  (Parachute|PushGliderOverride _playerpawn (Variables|Default|GetGlider%20Cosmetic%20Override) true)
  ;[1 1 1] Bind to the apex and deploy the glider
  (Variables|/Script/Engine.CharacterMovementComponent|SetbNotifyApex true (Variables|/Script/Engine.Character|GetCharacterMovement _playerpawn))
  (Character|BindEventtoOnReachedJumpApex _playerpawn (delegate Custom|Jump%20Apex%20Reached))
  (Variables|/Script/FortniteGame.FortPawn|SetbIgnoreNextFallingDamage true _playerpawn)
  (Utilities|FlowControl|Delay (Variables|Default|GetBadgerGrapeDelayFailsafe))
  (if (not (Variables|Default|GetHas%20Deployed))
    (Development|PrintString "FAILSAFE TRIGGERED" true true "(R=1.000000,G=0.214249,B=0.405303,A=1.000000)")
    (Utilities|FlowControl|DoOnce false
      (:Completed
        (bind _playerpawn_2 (Variables|Default|GetPlayerPawn)
          (:then
            (Variables|Default|SetHas%20Deployed true)
            (GameplayEffects|RemoveActiveEffectswithTags (Ability|GetAbilitySystemComponent _playerpawn_2) (Variables|Default|GetT_BlockingGE))
            (if (not (Class|AthenaFunctionLibrary|RowtoBool (Variables|Default|GetDisableGliderDeploy)))
              (Parachute|ForceOpenParachute _playerpawn_2 true)
              ;[1 0.375 0.375] Playing the open sound here because there's a problem in code with how we play the audio and fade/kill it on movement mode change. The sound is usually on the glider def and run in code, AFortPlayerParachute::SetAndPlayCurrentGliderSound().
              (Audio|SpawnSoundAttached (Variables|Default|GetSound_GliderOpen) (Variables|/Script/Engine.Actor|GetRootComponent (Variables|Default|GetPlayerPawn)))
              ;[1 1 1] Trigger the removal of the item if it's out of ammo.
              (Utilities|IsValid _weapon_actor
                (:Is%20Valid
                  (if (< (FortWeapon|GetRemainingAmmo _weapon_actor) 1)
                    (Ability|SendGameplayEventtoActor (Variables|Default|GetPlayerPawn) (Variables|Default|GetEvent_DestroyGrappler)))))))
          (:else))))))

(event Custom|Jump%20Apex%20Reached
  (bind _weapon_actor (Variables|Default|GetWeapon%20Actor))
  ;[1 1 1] Bind to the apex and deploy the glider
  (bind _rope (Variables|Default|GetRope)
    (:then
      (Rendering|SetActorHiddenInGame _rope true)
      (Development|PrintString "NATURAL DEPLOY EVENT TRIGGERED" true true "(R=1.000000,G=0.214249,B=0.405303,A=1.000000)")
      (Utilities|FlowControl|DoOnce false
        (:Completed
          (bind _playerpawn (Variables|Default|GetPlayerPawn)
            (:then
              (Variables|Default|SetHas%20Deployed true)
              (GameplayEffects|RemoveActiveEffectswithTags (Ability|GetAbilitySystemComponent _playerpawn) (Variables|Default|GetT_BlockingGE))
              (if (not (Class|AthenaFunctionLibrary|RowtoBool (Variables|Default|GetDisableGliderDeploy)))
                (Parachute|ForceOpenParachute _playerpawn true)
                ;[1 0.375 0.375] Playing the open sound here because there's a problem in code with how we play the audio and fade/kill it on movement mode change. The sound is usually on the glider def and run in code, AFortPlayerParachute::SetAndPlayCurrentGliderSound().
                (Audio|SpawnSoundAttached (Variables|Default|GetSound_GliderOpen) (Variables|/Script/Engine.Actor|GetRootComponent (Variables|Default|GetPlayerPawn)))
                ;[1 1 1] Trigger the removal of the item if it's out of ammo.
                (Utilities|IsValid _weapon_actor
                  (:Is%20Valid
                    (if (< (FortWeapon|GetRemainingAmmo _weapon_actor) 1)
                      (Ability|SendGameplayEventtoActor (Variables|Default|GetPlayerPawn) (Variables|Default|GetEvent_DestroyGrappler)))))))
            (:else)))))
    (:else)))

(event Custom|Teleported%20-%20Return%20abilities%20and%20stop
  (params (FortPawn :type Fort%20Pawn%20Object%20Reference))
  ;[1 0.05564 0] TELEPORT / RIFT FAIL SAFE
  (bind _playerpawn (Variables|Default|GetPlayerPawn)
    (:then
      (Parachute|PopGliderOverride _playerpawn (Variables|Default|GetGlider%20Cosmetic%20Override))
      (Utilities|Casting|CastToFortAbilitySystemComponent (Ability|GetAbilitySystemComponent _playerpawn))
      (GameplayEffects|RemoveActiveEffectswithTags (Ability|GetAbilitySystemComponent _playerpawn) (Variables|Default|GetT_BlockingGE))
      (Actor|DestroyActor))
    (:else)))

(event EventDestroyed
  ;[1 1 1] Destroyed - Also unblock ability tags and ensure glider override is cleared if the player never began skydiving (PopOverrideOnLand will ensure it's cleared if they did begin skydiving).
  (bind _playerpawn (Variables|Default|GetPlayerPawn)
    (:then
      (if (not (Parachute|IsSkydiving _playerpawn))
        (Parachute|PopGliderOverride _playerpawn (Variables|Default|GetGlider%20Cosmetic%20Override))))
    (:else
      _))
  (|CallParentFunction))

(fn ConstructionScript
  (|CallParentFunction))
